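# Physics asset for the cords on the LilSplit Sprinkles medium male player body, letting them swing and collide with the upper body.
# Two three-bone dynamic chains hang from master cord bodies linked to spine_05, with zero twist limit and much tighter swing on the left than the right.
# Capsules on upper spine and arms approximate the torso, joined by free angular constraints.
# The right chain's capsules are notably offset sideways from their bones.
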PhysicsAsset /BRCosmetics/Characters/Player/Male/Medium/Bodies/M_MED_LilSplit_Sprinkles/Meshes/M_MED_LilSplit_Sprinkles_Cord_Physics_FE
bodies {
 MSTR_cord_l [{shapeType=Capsule, length=5, radius=2, shapeName=MSTR_cord_l_capsule}]
 MSTR_cord_r [{shapeType=Capsule, length=5, radius=2, shapeName=MSTR_cord_r_capsule}]
 dyn_cord_l_01 [{shapeType=Capsule, length=9.5, radius=1, shapeName=dyn_cord_l_01_capsule, center=<4.2 0 0.9>, rotation=<rot -90 0 0>}]
 dyn_cord_l_02 [{shapeType=Capsule, length=3.5, radius=1, shapeName=dyn_cord_l_02_capsule, center=<2.2 0 0>, rotation=<rot -90 0 0>}]
 dyn_cord_l_03 [{shapeType=Capsule, length=3.5, radius=1, shapeName=dyn_cord_l_03_capsule, center=<2.2 0 0>, rotation=<rot -90 0 0>}]
 dyn_cord_r_01 [{shapeType=Capsule, length=6.5, radius=1, shapeName=dyn_cord_r_01_capsule, center=<4.8 0 -15>, rotation=<rot -90 0 0>}]
 dyn_cord_r_02 [{shapeType=Capsule, length=6.4, radius=1, shapeName=dyn_cord_r_02_capsule, center=<2.5 0 -15>, rotation=<rot -90 0 0>}]
 dyn_cord_r_03 [{shapeType=Capsule, length=6.4, radius=1, shapeName=dyn_cord_r_03_capsule, center=<2.5 0 -15>, rotation=<rot -90 0 0>}]
 lowerarm_l [{shapeType=Capsule, length=15, radius=5, shapeName=lowerarm_l_capsule, center=<5 0 0>, rotation=<rot -90 156.3706 -167.5188>}]
 lowerarm_r [{shapeType=Capsule, length=14, radius=5, shapeName=lowerarm_r_capsule, center=<-9 0 0>, rotation=<rot -90 0 0>}]
 spine_03 [
  {shapeType=Capsule, length=27.4992, radius=5.5008, shapeName=spine_03_capsule, center=<0 13.5 12>, rotation=<rot -90 2.2 1.5>}
  {shapeType=Capsule, length=30, radius=3, shapeName=spine_03_capsule, center=<0 13.5 -12>, rotation=<rot -90 4.3 -0.9>}
 ]
 spine_04 [{shapeType=Capsule, length=41, radius=20, shapeName=spine_04_capsule, center=<0 -0.2 0>, rotation=<rot -90 0 0>}]
 spine_05 [{shapeType=Capsule, length=8.5, radius=15.2, shapeName=spine_05_capsule, center=<0 3 -0.7>}]
 upperarm_l [{shapeType=Capsule, length=13.5, radius=5, shapeName=upperarm_l_capsule, center=<10.3 0 0>, rotation=<rot -90 0 0>}]
 upperarm_r [{
   shapeType=Capsule
   length=38.9937
   radius=11.8859
   shapeName=upperarm_l_capsule
   center=<-22.9072 -9.7074 -6.558>
   rotation=<rot -90 179.8881 0.4716>
  }]
}
constraints [
 {
  bone1Name=MSTR_cord_l
  bone2Name=spine_05
  swing1LimitDegrees=45
  swing1Motion=Free
  swing2LimitDegrees=45
  swing2Motion=Free
  twistLimitDegrees=45
  twistMotion=Free
 }
 {bone1Name=dyn_cord_l_01, bone2Name=MSTR_cord_l, swing1LimitDegrees=10, swing2LimitDegrees=10, twistLimitDegrees=0}
 {bone1Name=dyn_cord_l_02, bone2Name=dyn_cord_l_01, swing1LimitDegrees=10, swing2LimitDegrees=10, twistLimitDegrees=0}
 {bone1Name=dyn_cord_l_03, bone2Name=dyn_cord_l_02, swing1LimitDegrees=10, swing2LimitDegrees=10, twistLimitDegrees=0}
 {
  bone1Name=MSTR_cord_r
  bone2Name=spine_05
  swing1LimitDegrees=45
  swing1Motion=Free
  swing2LimitDegrees=45
  swing2Motion=Free
  twistLimitDegrees=45
  twistMotion=Free
 }
 {bone1Name=dyn_cord_r_01, bone2Name=MSTR_cord_r, swing1LimitDegrees=90, swing2LimitDegrees=90, twistLimitDegrees=0}
 {bone1Name=dyn_cord_r_02, bone2Name=dyn_cord_r_01, swing1LimitDegrees=90, swing2LimitDegrees=90, twistLimitDegrees=0}
 {bone1Name=dyn_cord_r_03, bone2Name=dyn_cord_r_02, swing1LimitDegrees=30, swing2LimitDegrees=30, twistLimitDegrees=0}
 {
  bone1Name=spine_04
  bone2Name=spine_03
  swing1LimitDegrees=45
  swing1Motion=Free
  swing2LimitDegrees=45
  swing2Motion=Free
  twistLimitDegrees=45
  twistMotion=Free
 }
 {
  bone1Name=upperarm_l
  bone2Name=spine_05
  swing1LimitDegrees=45
  swing1Motion=Free
  swing2LimitDegrees=45
  swing2Motion=Free
  twistLimitDegrees=45
  twistMotion=Free
 }
 {
  bone1Name=spine_05
  bone2Name=spine_04
  swing1LimitDegrees=45
  swing1Motion=Free
  swing2LimitDegrees=45
  swing2Motion=Free
  twistLimitDegrees=45
  twistMotion=Free
 }
 {
  bone1Name=upperarm_r
  bone2Name=spine_05
  swing1LimitDegrees=45
  swing1Motion=Free
  swing2LimitDegrees=45
  swing2Motion=Free
  twistLimitDegrees=45
  twistMotion=Free
 }
 {
  bone1Name=lowerarm_r
  bone2Name=upperarm_r
  swing1LimitDegrees=45
  swing1Motion=Free
  swing2LimitDegrees=45
  swing2Motion=Free
  twistLimitDegrees=45
  twistMotion=Free
 }
 {
  bone1Name=lowerarm_l
  bone2Name=upperarm_l
  swing1LimitDegrees=45
  swing1Motion=Free
  swing2LimitDegrees=45
  swing2Motion=Free
  twistLimitDegrees=45
  twistMotion=Free
 }
]
