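# Physics asset for the jacket of the male medium LilSplit Sprinkles player body, built entirely from capsules.
# Five jacket chains (A left/right, Ba left, C left/right) start at tiny MSTR root bodies constrained to spine_03.
# Large spine, pelvis and thigh capsules form the body volume, joined with free angular motion, as are the roots.
# Chain links have zero-degree twist limits and narrow swing, tightest on Ba and C, looser on A.
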PhysicsAsset /BRCosmetics/Characters/Player/Male/Medium/Bodies/M_MED_LilSplit_Sprinkles/Meshes/M_MED_LilSplit_Sprinkles_Jack_Physics
bodies {
 AD_jacketA_r_03 [{shapeType=Capsule, length=5, radius=4, shapeName=AD_jacketA_r_03_capsule, center=<2.5 5 0>, rotation=<rot -90 0 0>}]
 CR_jacketA_r_02 [{shapeType=Capsule, length=5, radius=4, shapeName=CR_jacketA_r_02_capsule, center=<0 3.7 0>, rotation=<rot -89.2967 -36.3189 38.5908>}]
 CR_jacketBa_l_01 [{shapeType=Capsule, length=6, radius=4, shapeName=CR_jacketBa_l_01_capsule, center=<10.5 1.5 0>, rotation=<rot -90 0 0>}]
 CR_jacketBa_l_02 [{shapeType=Capsule, length=6, radius=4, shapeName=CR_jacketBa_l_02_capsule, center=<10.5 1.5 0>, rotation=<rot -90 0 0>}]
 CR_jacketC_l_01 [{shapeType=Capsule, length=6, radius=4, shapeName=CR_jacketC_l_01_capsule, center=<10.5 1.3 0>, rotation=<rot -90 0 0>}]
 CR_jacketC_l_02 [{shapeType=Capsule, length=6, radius=4, shapeName=CR_jacketC_l_02_capsule, center=<10.5 4 0>, rotation=<rot -90 0 0>}]
 CR_jacketC_r_01 [{shapeType=Capsule, length=6, radius=4, shapeName=CR_jacketC_r_01_capsule, center=<10.5 1.3 0>, rotation=<rot -90 0 0>}]
 CR_jacketC_r_02 [{shapeType=Capsule, length=6, radius=4, shapeName=CR_jacketC_r_02_capsule, center=<10.5 4 0>, rotation=<rot -90 0 0>}]
 MSTR_jacketA_l [{shapeType=Capsule, length=0.505, radius=0.505, shapeName=MSTR_jacketA_l_capsule, rotation=<rot 0 -90 0>}]
 MSTR_jacketA_r [{shapeType=Capsule, length=0.505, radius=0.505, shapeName=MSTR_jacketA_r_capsule, rotation=<rot 0 -90 0>}]
 MSTR_jacketBa_l [{shapeType=Capsule, length=0.505, radius=0.505, shapeName=MSTR_jacketBa_l_capsule, rotation=<rot 0 -90 0>}]
 MSTR_jacketC_l [{shapeType=Capsule, length=0.505, radius=0.505, shapeName=MSTR_jacketC_l_capsule, rotation=<rot 0 -90 0>}]
 MSTR_jacketC_r [{shapeType=Capsule, length=0.505, radius=0.505, shapeName=MSTR_jacketC_r_capsule, rotation=<rot 0 -90 0>}]
 RB_AD_jacketA_l_03 [{shapeType=Capsule, length=7, radius=5, shapeName=RB_AD_jacketA_l_03_capsule, center=<3 7.5 0>, rotation=<rot -90 180 180>}]
 RB_jacketA_l_01 [{shapeType=Capsule, length=7, radius=5, shapeName=RB_jacketA_l_01_capsule, center=<3 4.1 0>, rotation=<rot -90 0 0>}]
 RB_jacketA_l_02 [{shapeType=Capsule, length=7, radius=5, shapeName=RB_jacketA_l_02_capsule, center=<3 5 0>, rotation=<rot -90 0 0>}]
 pelvis [{shapeType=Capsule, length=9.5, radius=15, shapeName=pelvis_capsule, center=<0 -1.0643 0>}]
 spine_03 [{shapeType=Capsule, length=31.2812, radius=15.8918, shapeName=spine_03_capsule, rotation=<rot -90 0 0>}]
 thigh_l [{
   shapeType=Capsule
   length=38.908
   radius=11.0497
   shapeName=thigh_l_capsule
   center=<-21.5273 0.3659 1.7369>
   rotation=<rot 11.0802 -77.9958 -87.3743>
  }]
 thigh_r [{
   shapeType=Capsule
   length=25.4713
   radius=15.8249
   shapeName=thigh_r_capsule
   center=<21.236 2.7524 -8.2787>
   rotation=<rot 20.3155 -83.1855 -81.4892>
  }]
}
constraints [
 {bone1Name=CR_jacketA_r_02, bone2Name=MSTR_jacketA_r, swing1LimitDegrees=30, swing2LimitDegrees=30, twistLimitDegrees=0}
 {bone1Name=AD_jacketA_r_03, bone2Name=CR_jacketA_r_02, swing1LimitDegrees=5, swing2LimitDegrees=5, twistLimitDegrees=0}
 {
  bone1Name=MSTR_jacketA_r
  bone2Name=spine_03
  swing1LimitDegrees=45
  swing1Motion=Free
  swing2LimitDegrees=45
  swing2Motion=Free
  twistLimitDegrees=45
  twistMotion=Free
 }
 {bone1Name=RB_AD_jacketA_l_03, bone2Name=RB_jacketA_l_02, swing1LimitDegrees=30, swing2LimitDegrees=30, twistLimitDegrees=0}
 {
  bone1Name=MSTR_jacketA_l
  bone2Name=spine_03
  swing1LimitDegrees=45
  swing1Motion=Free
  swing2LimitDegrees=45
  swing2Motion=Free
  twistLimitDegrees=45
  twistMotion=Free
 }
 {bone1Name=RB_jacketA_l_01, bone2Name=MSTR_jacketA_l, swing1LimitDegrees=30, swing2LimitDegrees=30, twistLimitDegrees=0}
 {bone1Name=RB_jacketA_l_02, bone2Name=RB_jacketA_l_01, swing1LimitDegrees=30, swing2LimitDegrees=30, twistLimitDegrees=0}
 {
  bone1Name=MSTR_jacketBa_l
  bone2Name=spine_03
  swing1LimitDegrees=45
  swing1Motion=Free
  swing2LimitDegrees=45
  swing2Motion=Free
  twistLimitDegrees=45
  twistMotion=Free
 }
 {bone1Name=CR_jacketBa_l_01, bone2Name=MSTR_jacketBa_l, swing1LimitDegrees=3, swing2LimitDegrees=3, twistLimitDegrees=0}
 {bone1Name=CR_jacketBa_l_02, bone2Name=CR_jacketBa_l_01, swing1LimitDegrees=3, swing2LimitDegrees=3, twistLimitDegrees=0}
 {
  bone1Name=MSTR_jacketC_l
  bone2Name=spine_03
  swing1LimitDegrees=45
  swing1Motion=Free
  swing2LimitDegrees=45
  swing2Motion=Free
  twistLimitDegrees=45
  twistMotion=Free
 }
 {bone1Name=CR_jacketC_l_01, bone2Name=MSTR_jacketC_l, swing1LimitDegrees=3, swing2LimitDegrees=3, twistLimitDegrees=0}
 {bone1Name=CR_jacketC_l_02, bone2Name=CR_jacketC_l_01, swing1LimitDegrees=3, swing2LimitDegrees=3, twistLimitDegrees=0}
 {
  bone1Name=MSTR_jacketC_r
  bone2Name=spine_03
  swing1LimitDegrees=45
  swing1Motion=Free
  swing2LimitDegrees=45
  swing2Motion=Free
  twistLimitDegrees=45
  twistMotion=Free
 }
 {bone1Name=CR_jacketC_r_01, bone2Name=MSTR_jacketC_r, swing1LimitDegrees=3, swing2LimitDegrees=3, twistLimitDegrees=0}
 {bone1Name=CR_jacketC_r_02, bone2Name=CR_jacketC_r_01, swing1LimitDegrees=3, swing2LimitDegrees=3, twistLimitDegrees=0}
 {
  bone1Name=spine_03
  bone2Name=pelvis
  swing1LimitDegrees=45
  swing1Motion=Free
  swing2LimitDegrees=45
  swing2Motion=Free
  twistLimitDegrees=45
  twistMotion=Free
 }
 {
  bone1Name=thigh_l
  bone2Name=pelvis
  swing1LimitDegrees=45
  swing1Motion=Free
  swing2LimitDegrees=45
  swing2Motion=Free
  twistLimitDegrees=45
  twistMotion=Free
 }
 {
  bone1Name=thigh_r
  bone2Name=pelvis
  swing1LimitDegrees=45
  swing1Motion=Free
  swing2LimitDegrees=45
  swing2Motion=Free
  twistLimitDegrees=45
  twistMotion=Free
 }
]
