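# Physics asset for the face accessory part of the LilSplit Sprinkles male medium character cosmetic, covering its cream base, earphones and glasses.
# Each accessory pairs a MSTR bone capsule with a dyn bone capsule, joined by one constraint.
# Constraints allow only a few degrees of swing with zero twist, on a single axis for earphones and glasses and two for the cream base, suggesting subtle secondary jiggle.
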PhysicsAsset /BRCosmetics/Characters/Player/Male/Medium/Bodies/M_MED_LilSplit_Sprinkles/Meshes/Parts/M_MED_LilSplit_Sprinkles_FaceAcc_Physics
bodies {
 MSTR_creamBase [{shapeType=Capsule, length=0.505, radius=0.505, shapeName=MSTR_creamBase_capsule, rotation=<rot 0 -90 0>}]
 MSTR_earphones [{shapeType=Capsule, length=3, radius=2, shapeName=MSTR_earphones_capsule}]
 MSTR_glasses [{shapeType=Capsule, length=3, radius=2, shapeName=MSTR_glasses_capsule}]
 dyn_creamBase_01 [{shapeType=Capsule, length=3.4603, radius=5.705, shapeName=dyn_creamBase_01_capsule, center=<4.1766 0 0>, rotation=<rot -90 180 180>}]
 dyn_earphones_01 [{shapeType=Capsule, length=5, radius=4.8188, shapeName=dyn_earphones_01_capsule, center=<3.2 0 0>, rotation=<rot -90 0 0>}]
 dyn_glasses_01 [{shapeType=Capsule, length=3, radius=2, shapeName=dyn_glasses_01_capsule, center=<4.2252 0 0>, rotation=<rot -90 0 0>}]
}
constraints [
 {bone1Name=dyn_earphones_01, bone2Name=MSTR_earphones, swing1LimitDegrees=3, swing2LimitDegrees=0, twistLimitDegrees=0}
 {bone1Name=dyn_glasses_01, bone2Name=MSTR_glasses, swing1LimitDegrees=10, swing2LimitDegrees=0, twistLimitDegrees=0}
 {bone1Name=dyn_creamBase_01, bone2Name=MSTR_creamBase, swing1LimitDegrees=5, swing2LimitDegrees=5, twistLimitDegrees=0}
]
